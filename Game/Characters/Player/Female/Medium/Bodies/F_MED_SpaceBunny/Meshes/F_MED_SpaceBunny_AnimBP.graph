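# Anim Blueprint for the F_MED_SpaceBunny player body that applies corrective shoulder bone adjustments as the right arm moves.
# Each update maps upperarm_r yaw, relative to spine_05's reference pose, to arm_up_r and arm_up_more_r.
# AnimGraph copies the parent mesh pose, then arm_up_r drives interpolated, additive bone-space offsets on left shoulder armor and right front shoulder bones.
# Targeting events set ads_alpha, which, like arm_up_more_r, the AnimGraph never reads.
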AnimBlueprint /Game/Characters/Player/Female/Medium/Bodies/F_MED_SpaceBunny/Meshes/F_MED_SpaceBunny_AnimBP
target_skeleton=&/F_MED_SpaceBunny_Skeleton
graphs [
 AnimGraph {nodes [
   n1 = Get arm_up_r() -> (arm_up_r:double)
   n2 = "Transform (Modify) Bone - Bone: dyn_shoulder_armor_LOCALSPACE_l"(
    ComponentPose=Local To Component(LocalPose=Copy Pose From Mesh() {bUseAttachedParent=true})
    Alpha=@n1
    Translation="2.75,-0.000,0.435"
    Rotation="0,0,0"
    Scale="1,1,1"
   ) {
    translationMode=BMM_Additive
    translationSpace=BCS_BoneSpace
    alphaBoolBlend {blendOption=Linear}
    alphaScaleBiasClamp {bInterpResult=true, interpSpeedDecreasing=25, interpSpeedIncreasing=35}
    boneToModify {boneName=dyn_shoulder_armor_LOCALSPACE_l}
    translation {x=2.75, z=0.435}
   }
   n3 = "Transform (Modify) Bone - Bone: dyn_shoulder_armor_driver_l"(ComponentPose=@n2, Alpha=@n1, Translation="0,0,0", Rotation="-5,0.000,0.000", Scale="1,1,1") {
    rotationMode=BMM_Additive
    rotationSpace=BCS_BoneSpace
    translationSpace=BCS_BoneSpace
    alphaBoolBlend {blendOption=Linear}
    alphaScaleBiasClamp {bInterpResult=true, interpSpeedDecreasing=25, interpSpeedIncreasing=35}
    boneToModify {boneName=dyn_shoulder_armor_driver_l}
    rotation {pitch=-5}
   }
   n4 = "Transform (Modify) Bone - Bone: dyn_shoulder_frt_r"(ComponentPose=@n3, Alpha=@n1, Translation="0.486,1.557,0.040", Rotation="9.944,-17.602,-1.452", Scale="0.6,0.6,0.6") {
    rotationMode=BMM_Additive
    rotationSpace=BCS_BoneSpace
    scaleMode=BMM_Additive
    scaleSpace=BCS_BoneSpace
    translationMode=BMM_Additive
    translationSpace=BCS_BoneSpace
    alphaBoolBlend {blendOption=Linear}
    alphaScaleBiasClamp {bInterpResult=true, interpSpeedDecreasing=25, interpSpeedIncreasing=35}
    boneToModify {boneName=dyn_shoulder_frt_r}
    rotation=<rot 9.944 -17.602 -1.452>
    scale=<0.6 0.6 0.6>
    translation=<0.486 1.557 0.04>
   }
   Output Pose(Result=Component To Local(ComponentPose=@n4)) {layerGroup=None}
  ]}
 EventGraph {nodes [
   Event BlueprintUpdateAnimation()
   > Sequence()
   n1 = GetOwnerSkeletalMeshComponent()
   n2 = BreakRotator(InRot=BreakTransform(InTransform=GetDeltaTransformFromRefPose(self=@n1, BoneName=upperarm_r, BaseName=spine_05)).Rotation)
   > Set arm_up_r(arm_up_r:double=MapRangeClamped(Value=@n2.Yaw, InRangeA=80, InRangeB=138)) -> (Output_Get:double)
   > Set arm_up_more_r(arm_up_more_r:double=MapRangeClamped(Value=@n2.Yaw, InRangeA=85, InRangeB=125)) -> (Output_Get:double)
   Event OnBeginTargeting() {is=title K2Node_Event}
   > Set ads_alpha(ads_alpha:double=1) -> (Output_Get:double)
   Event OnEndTargeting() {is=title K2Node_Event}
   > Set ads_alpha(ads_alpha=0)
  ]}
]
variables [
 ads_alpha {type=number}
 arm_up_more_r {type=number}
 arm_up_r {type=number}
]
